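; Foul Tame head Blueprint that turns its skeletal mesh to face a pre- or post-aircraft rotation depending on the Fortnite game phase.
; BeginPlay sets the water material collection's YachtLocation so ripples show, then listens to the stateful game phase event, handling saved and new states.
; GamePhaseChanged rotates the mesh and its water-contact Niagara effect, offset 180 degrees.
; EndPlay stops the listener; burp, overlap and tick events are empty.
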(event EventBeginPlay
  ;[1 1 1] setting this so when present in the water ripples show.
  (|CallParentFunction)
  (Rendering|Material|SetVectorParameterValue "/Water/Materials/MPC/MPC_Water.MPC_Water" "YachtLocation" (Utilities|Struct|MakeLinearColor 1.0) :!class /Script/Engine.KismetMaterialLibrary)
  (Networking|SwitchHasAuthority
    (:Authority
      _)
    (:Remote
      _))
  (bind _returnvalue (GameplayEventRouter|GetEventRouter "/Script/FortniteGame.FortGameState" self))
  (bind (_asynctaskproxy _payload) (GameplayEventRouter|StartListeningToStatefulEvent _returnvalue "/Script/FortniteGame.FortGamePhaseUpdatedEvent")
    (:then
      (Variables|Default|SetGamePhaseUpdatedAsyncTask _asynctaskproxy))
    (:OnEventReceived
      _)
    (:OnSavedEventStateExists
      _))
  (CallFunction|GamePhaseChanged (Utilities|Struct|BreakFortGamePhaseUpdatedEvent _payload)))

(event Custom|GamePhaseChanged
  (params (GamePhase :type EAthenaGamePhase%20Enum))
  (bind _postaircraftrotation (Variables|Default|GetPostAircraftRotation))
  (bind _skeletalmesh (Variables|Default|GetSkeletalMesh))
  (bind _preaircraftrotation (Variables|Default|GetPreAircraftRotation))
  (bind _ns_foultame_watercontact (Variables|Default|GetNS_FoulTame_WaterContact))
  (switch Utilities|FlowControl|Switch|SwitchonEAthenaGamePhase GamePhase
    (:None
      _)
    (:Setup
      _)
    (:Warmup
      _)
    (:Aircraft
      (Transformation|SetRelativeRotation _skeletalmesh (Math|Rotator|MakeRotator 0.0 0.0 _postaircraftrotation) false false :!label _label)
      (Transformation|SetRelativeRotation _ns_foultame_watercontact (Math|Rotator|MakeRotator 0.0 0.0 (+ _postaircraftrotation 1.8e+02)) false false))
    (:SafeZones
      (goto _label))
    (:EndGame
      (goto _label)))
  (Transformation|SetRelativeRotation _skeletalmesh (Math|Rotator|MakeRotator 0.0 0.0 _preaircraftrotation) false false)
  (Transformation|SetRelativeRotation _ns_foultame_watercontact (Math|Rotator|MakeRotator 0.0 0.0 (+ _preaircraftrotation 1.8e+02)) false false))

(event EventEndPlay
  (params (EndPlayReason :type EEndPlayReason%20Enum :default "Destroyed"))
  (|CallParentFunction EndPlayReason)
  (bind _gamephaseupdatedasynctask (Variables|Default|GetGamePhaseUpdatedAsyncTask)
    (:then
      (GameplayEventRouter|StopListeningtoStatefulEvent _gamephaseupdatedasynctask))
    (:else)))

(event Custom|StartBurp)

(event Custom|EndBurp)

(event Collision|EventActorBeginOverlap
  (params (OtherActor :type Actor%20Object%20Reference)))

(event EventTick
  (params (DeltaSeconds :type Float%20%28single-precision%29)))

(fn ConstructionScript)
