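# Animation Blueprint for the FoulTame NPC head on the BuzzCool NPC skeleton, playing an idle with occasional squints beneath DefaultSlot.
# A nested state machine loops the base idle, with a dynamic additive built from the bobbing clip and its time-zero pose layered on top.
# Entering base idle fires a notify that waits a stream-random delay, then pulses idleBreaker for one tick, triggering a one-shot squint that returns automatically.
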AnimBlueprint /WildEstate_FoulTame/Gameplay/M_MED_BuzzCool_NPCHead_AnimBP
target_skeleton=/BuzzCool/NPCs/BuzzCool_Clones/Meshes/M_MED_BuzzCool_NPC_Skeleton
graphs [
 AnimGraph {nodes [
   AnimGraphNode_StateMachine_0 = fullbody() {is=title AnimGraphNode_StateMachine, bAllowConduitEntryStates=true}
   Output Pose(Result=Slot 'DefaultSlot'(Source=@AnimGraphNode_StateMachine_0))
  ]}
 fullbody {entry=idle, path=AnimGraph.AnimGraphNode_StateMachine_0.fullbody, ids=id, nodes [AnimStateNode_0 = idle() {is=title AnimStateNode}]}
 idle {
  path=AnimGraph.AnimGraphNode_StateMachine_0.fullbody.AnimStateNode_0.idle
  nodes [
   Random Sequence Player!() -> (Pose:Pose Link Structure) {
    is=title AnimGraphNode_RandomPlayer
    entries [
     {
      chanceToPlay=1
      maxLoopCount=0
      maxPlayRate=1
      minLoopCount=0
      minPlayRate=1
      sequence=/WildEstate_FoulTame/Assets/Animation/Actions/FoulTame_Idle_Base
      blendIn {blendOption=HermiteCubic, blendTime=0.5, customCurve=null}
     }
     {
      chanceToPlay=0.2
      maxLoopCount=0
      maxPlayRate=1
      minLoopCount=0
      minPlayRate=1
      sequence=/WildEstate_FoulTame/Assets/Animation/Actions/FoulTame_Idle_Squint
      blendIn {blendOption=HermiteCubic, blendTime=0.5, customCurve=null}
     }
    ]
   }
   n1 = "Sequence Evaluator (by time) 'FoulTame_Bobbing'"!(ExplicitTime:float=0) -> (Pose:Pose Link Structure) {is=title AnimGraphNode_SequenceEvaluator, sequence=/WildEstate_FoulTame/Assets/Animation/Actions/FoulTame_Bobbing}
   n2 = Sequence Player 'FoulTame_Bobbing'() {sequence=&FoulTame_Bobbing}
   AnimGraphNode_StateMachine_0 = idle_SM() {is=title AnimGraphNode_StateMachine, bAllowConduitEntryStates=true}
   Output Animation Pose(Result=Apply Additive(
     Base=@AnimGraphNode_StateMachine_0
     Additive=Make Dynamic Additive!(Base:Pose Link Structure=@n2, Additive:Pose Link Structure=@n1) -> (Pose:Pose Link Structure) {is=title AnimGraphNode_MakeDynamicAdditive}
    ))
  ]
 }
 idle_SM {
  entry=baseIdle
  path=AnimGraph.AnimGraphNode_StateMachine_0.fullbody.AnimStateNode_0.idle.AnimGraphNode_StateMachine_0.idle_SM
  ids=id
  nodes [
   AnimStateNode_0 = baseIdle() {is=title AnimStateNode, stateEntered {guid=#1, notifyName=idleBreakerResetNotif}}
   AnimStateNode_1 = idleSquint() {is=title AnimStateNode}
   AnimStateTransitionNode_0 = AnimStateTransitionNode() {bAllowInertializationForSelfTransitions=true}
   AnimStateTransitionNode_1 = AnimStateTransitionNode() {bAllowInertializationForSelfTransitions=true, bAutomaticRuleBasedOnSequencePlayerInState=true}
  ]
  links [
   {
    dst=idleSquint
    rule=AnimGraph.AnimGraphNode_StateMachine_0.fullbody.AnimStateNode_0.idle.AnimGraphNode_StateMachine_0.idle_SM.AnimStateTransitionNode_0.Transition
    src=baseIdle
    via=AnimStateTransitionNode_0
   }
   {
    dst=baseIdle
    rule=AnimGraph.AnimGraphNode_StateMachine_0.fullbody.AnimStateNode_0.idle.AnimGraphNode_StateMachine_0.idle_SM.AnimStateTransitionNode_1.Transition
    src=idleSquint
    via=AnimStateTransitionNode_1
   }
  ]
 }
 baseIdle {
  path=AnimGraph.AnimGraphNode_StateMachine_0.fullbody.AnimStateNode_0.idle.AnimGraphNode_StateMachine_0.idle_SM.AnimStateNode_0.baseIdle
  nodes [Output Animation Pose(Result=Sequence Player 'FoulTame_Idle_Base'() {sequence=&FoulTame_Idle_Base})]
 }
 idleSquint {
  path=AnimGraph.AnimGraphNode_StateMachine_0.fullbody.AnimStateNode_0.idle.AnimGraphNode_StateMachine_0.idle_SM.AnimStateNode_1.idleSquint
  nodes [Output Animation Pose(Result=Sequence Player 'FoulTame_Idle_Squint'() {bLoopAnimation=false, sequence=&FoulTame_Idle_Squint})]
 }
 Transition {
  path=AnimGraph.AnimGraphNode_StateMachine_0.fullbody.AnimStateNode_0.idle.AnimGraphNode_StateMachine_0.idle_SM.AnimStateTransitionNode_0.Transition
  nodes [Result(bCanEnterTransition=Get idleBreaker() -> (idleBreaker:bool))]
 }
 Transition {
  path=AnimGraph.AnimGraphNode_StateMachine_0.fullbody.AnimStateNode_0.idle.AnimGraphNode_StateMachine_0.idle_SM.AnimStateTransitionNode_1.Transition
  nodes [Result()]
 }
 EventGraph {nodes [
   AnimNotify_idleBreakerResetNotif() {is=title K2Node_Event}
   n1 = Get!() -> (idleBreakerTimeRange_X:double, idleBreakerTimeRange_Y:double)
   > Set idleBreakerTime(idleBreakerTime:double=RandomFloatInRangeFromStream!(Stream_InitialSeed:int=0, Min:float=@n1, Max:float=@n1.idleBreakerTimeRange_Y) -> (ReturnValue:float)) -> (Output_Get:double)
   > Set idleBreaker(idleBreaker:bool=false) -> (Output_Get:bool)
   > RetriggerableDelay(Duration=Get idleBreakerTime() -> (idleBreakerTime:double))
   > Set idleBreaker(idleBreaker=true)
   > DelayUntilNextTick!(execute) -> (then)
   > Set idleBreaker(idleBreaker=false)
  ]}
 OnStateEntry {nodes [
   OnStateEntry!() -> (then, Context:"Anim Update Context Structure (by ref)", Node:"Anim Node Reference Structure (by ref)") {is=title K2Node_FunctionEntry}
   Return Node()
  ]}
]
variables [
 idleBreaker {type=boolean}
 idleBreakerTime {type=number}
 idleBreakerTimeRange {type=Vector2D, default=<10 15>}
]
